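# Physics asset for the male medium StallionSmoke jacket (FE variant), providing bodies for simulating the jacket's hanging panels.
# Seven four-capsule chains around the back, sides and front hang from small master bodies constrained to the pelvis, with limited swing and tight twist.
# Free-rotation constraints cross-link some neighboring chains, tying panels together.
# Pelvis and leg capsules supply body shapes, and a huge offset sphere on root, named floorCOL, approximates the floor.
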PhysicsAsset /BRCosmetics/Characters/Player/Male/Medium/Bodies/M_MED_StallionSmoke_Jacket/Meshes/M_MED_StallionSmoke_Jacket_FE_Physics
bodies {
 calf_l [{shapeType=Capsule, length=34.6821, radius=12, center=<-25 -2 0>, rotation=<rot 90 0 0>}]
 calf_r [{shapeType=Capsule, length=27.9321, radius=12, center=<25 0 0.1223>, rotation=<rot 90 0 0>}]
 dyn_jacket_b_01 [{shapeType=Capsule, length=3.75, radius=6, shapeName=dyn_jacket_b_l_01_capsule, center=<3 0 6>, rotation=<rot 0 0 90>}]
 dyn_jacket_b_02 [{shapeType=Capsule, length=12.5, radius=10, shapeName=dyn_jacket_b_l_02_capsule, center=<3 0 6>, rotation=<rot 0 0 90>}]
 dyn_jacket_b_03 [{shapeType=Capsule, length=12, radius=9, shapeName=dyn_jacket_b_l_03_capsule, center=<3 0 8>, rotation=<rot 0 0 90>}]
 dyn_jacket_b_04 [{shapeType=Capsule, length=24, radius=9, shapeName=dyn_jacket_b_l_04_capsule, center=<3 0 6>, rotation=<rot 0 0 90>}]
 dyn_jacket_b_l_01 [{shapeType=Capsule, length=3.75, radius=6, shapeName=dyn_jacket_b_l_01_capsule, center=<3 0 5>, rotation=<rot 0 0 90>}]
 dyn_jacket_b_l_02 [{shapeType=Capsule, length=3.5, radius=8.5, shapeName=dyn_jacket_b_l_02_capsule, center=<4 0 5>, rotation=<rot 0 0 90>}]
 dyn_jacket_b_l_03 [{shapeType=Capsule, length=3.5, radius=8.5, shapeName=dyn_jacket_b_l_03_capsule, center=<3 0 5>, rotation=<rot 0 0 90>}]
 dyn_jacket_b_l_04 [{shapeType=Capsule, length=12, radius=9, shapeName=dyn_jacket_b_l_04_capsule, center=<3 0 5>, rotation=<rot 0 0 90>}]
 dyn_jacket_b_r_01 [{shapeType=Capsule, length=4, radius=5, shapeName=dyn_jacket_b_r_01_capsule, center=<3 0 6>, rotation=<rot 0 0 90>}]
 dyn_jacket_b_r_02 [{shapeType=Capsule, length=4, radius=8, shapeName=dyn_jacket_b_r_02_capsule, center=<3 0 6>, rotation=<rot 0 0 90>}]
 dyn_jacket_b_r_03 [{shapeType=Capsule, length=10, radius=9, shapeName=dyn_jacket_b_r_03_capsule, center=<3 0 6>, rotation=<rot 0 0 90>}]
 dyn_jacket_b_r_04 [{shapeType=Capsule, length=19.5, radius=9, shapeName=dyn_jacket_b_r_04_capsule, center=<3 0 3>, rotation=<rot 0 0 90>}]
 dyn_jacket_f_l_01 [{shapeType=Capsule, length=6.25, radius=6.75, shapeName=dyn_jacket_f_l_01_capsule, center=<3 0 6>, rotation=<rot 0 0 90>}]
 dyn_jacket_f_l_02 [{shapeType=Capsule, length=7, radius=8.5, shapeName=dyn_jacket_f_l_02_capsule, center=<3 0 6>, rotation=<rot 0 0 90>}]
 dyn_jacket_f_l_03 [{shapeType=Capsule, length=5.5, radius=8, shapeName=dyn_jacket_f_l_03_capsule, center=<3 0 6>, rotation=<rot 0 0 90>}]
 dyn_jacket_f_l_04 [{shapeType=Capsule, length=12, radius=9, shapeName=dyn_jacket_f_l_04_capsule, center=<3 0 6>, rotation=<rot 0 0 90>}]
 dyn_jacket_f_r_01 [{shapeType=Capsule, length=6.25, radius=6.75, shapeName=dyn_jacket_f_l_01_capsule, center=<3 0 6>, rotation=<rot 0 0 90>}]
 dyn_jacket_f_r_02 [{shapeType=Capsule, length=14.5, radius=8, shapeName=dyn_jacket_f_l_02_capsule, center=<3 0 6>, rotation=<rot 0 0 90>}]
 dyn_jacket_f_r_03 [{shapeType=Capsule, length=3.25, radius=8.75, shapeName=dyn_jacket_f_r_03_capsule, center=<3 0 6>, rotation=<rot 0 0 90>}]
 dyn_jacket_f_r_04 [{shapeType=Capsule, length=12, radius=9, shapeName=dyn_jacket_f_l_04_capsule, center=<3 0 6>, rotation=<rot 0 0 90>}]
 dyn_jacket_s_l_01 [{shapeType=Capsule, length=6.25, radius=6.75, shapeName=dyn_jacket_s_l_01_capsule, center=<3 0 6>, rotation=<rot 0 0 90>}]
 dyn_jacket_s_l_02 [{shapeType=Capsule, length=5.25, radius=8.5, shapeName=dyn_jacket_s_l_02_capsule, center=<4 0 6>, rotation=<rot 0 0 90>}]
 dyn_jacket_s_l_03 [{shapeType=Capsule, length=4, radius=9, shapeName=dyn_jacket_s_l_03_capsule, center=<3 0 8>, rotation=<rot 0 0 90>}]
 dyn_jacket_s_l_04 [{shapeType=Capsule, length=12, radius=9, shapeName=dyn_jacket_s_l_04_capsule, center=<3 0 6>, rotation=<rot 0 0 90>}]
 dyn_jacket_s_r_01 [{shapeType=Capsule, length=6.25, radius=6.75, shapeName=dyn_jacket_s_l_01_capsule, center=<3 0 6.5>, rotation=<rot 0 0 90>}]
 dyn_jacket_s_r_02 [{shapeType=Capsule, length=5.25, radius=6, shapeName=dyn_jacket_s_l_02_capsule, center=<3 0 6>, rotation=<rot 0 0 90>}]
 dyn_jacket_s_r_03 [{shapeType=Capsule, length=4, radius=8, shapeName=dyn_jacket_s_r_03_capsule, center=<3 0 8>, rotation=<rot 0 0 90>}]
 dyn_jacket_s_r_04 [{shapeType=Capsule, length=12, radius=9, shapeName=dyn_jacket_s_r_04_capsule, center=<3 0 7>, rotation=<rot 0 0 90>}]
 dyn_mstr_jacket_b [{shapeType=Capsule, length=0.505, radius=0.505, shapeName=dyn_mstr_jacket_b_l_capsule, rotation=<rot 0 -90 0>}]
 dyn_mstr_jacket_b_l [{shapeType=Capsule, length=0.505, radius=0.505, shapeName=dyn_mstr_jacket_b_l_capsule, rotation=<rot 0 -90 0>}]
 dyn_mstr_jacket_b_r [{shapeType=Capsule, length=0.505, radius=0.505, shapeName=dyn_mstr_jacket_b_r_capsule, rotation=<rot 0 -90 0>}]
 dyn_mstr_jacket_f_l [{shapeType=Capsule, length=0.505, radius=0.505, shapeName=dyn_mstr_jacket_f_l_capsule, rotation=<rot 0 -90 0>}]
 dyn_mstr_jacket_f_r [{shapeType=Capsule, length=0.505, radius=0.505, shapeName=dyn_mstr_jacket_f_r_capsule, rotation=<rot 0 -90 0>}]
 dyn_mstr_jacket_s_l [{shapeType=Capsule, length=0.505, radius=0.505, shapeName=dyn_mstr_jacket_s_l_capsule, rotation=<rot 0 -90 0>}]
 dyn_mstr_jacket_s_r [{shapeType=Capsule, length=0.505, radius=0.505, shapeName=dyn_mstr_jacket_s_r_capsule, rotation=<rot 0 -90 0>}]
 foot_l [{shapeType=Capsule, length=4.25, radius=13, shapeName=foot_l_capsule, center=<-5.9857 6.4408 0>, rotation=<rot 0 0 90>}]
 foot_r [{shapeType=Capsule, length=16, radius=16, shapeName=foot_r_capsule, center=<3.3566 -4.826 0>, rotation=<rot 0 0 90>}]
 pelvis [{shapeType=Capsule, length=16.2569, radius=12.3526, center=<0 1.1827 0>, rotation=<rot 90 0 0>}]
 root [{shapeType=Sphere, radius=200, shapeName=floorCOL_sphere, center=<0 15 -180>}]
 thigh_l [
  {
   shapeType=Capsule
   length=31.5976
   radius=10
   shapeName=thigh_l_capsule
   center=<-15.134 -0.7627 -0.2574>
   rotation=<rot 90 89.6476 -86.6532>
  }
  {
   shapeType=Capsule
   length=5.25
   radius=8.75
   shapeName=thigh_l_capsule
   center=<-11.7777 -1.0295 -1.0619>
   rotation=<rot 34.6347 -33.1365 -43.8414>
  }
 ]
 thigh_r [
  {shapeType=Capsule, length=35.0976, radius=10.5, shapeName=thigh_l_capsule, center=<18.8655 2.0961 1.0827>, rotation=<rot 90 180 0>}
  {shapeType=Capsule, length=8, radius=9, shapeName=thigh_r_capsule, center=<11.7777 1.0295 1.0619>, rotation=<rot 34.6347 -33.1365 -43.8414>}
 ]
}
constraints [
 {bone1Name=dyn_jacket_b_l_04, bone2Name=dyn_jacket_b_l_03, swing1LimitDegrees=45, swing2LimitDegrees=45, twistLimitDegrees=3}
 {bone1Name=dyn_jacket_b_l_01, bone2Name=dyn_mstr_jacket_b_l, swing1LimitDegrees=20, swing2LimitDegrees=20, twistLimitDegrees=3}
 {bone1Name=dyn_jacket_b_l_02, bone2Name=dyn_jacket_b_l_01, swing1LimitDegrees=20, swing2LimitDegrees=20, twistLimitDegrees=3}
 {bone1Name=dyn_jacket_b_l_03, bone2Name=dyn_jacket_b_l_02, swing1LimitDegrees=30, swing2LimitDegrees=30, twistLimitDegrees=3}
 {bone1Name=dyn_jacket_b_r_04, bone2Name=dyn_jacket_b_r_03, swing1LimitDegrees=15, swing2LimitDegrees=70, twistLimitDegrees=3}
 {bone1Name=dyn_jacket_b_r_03, bone2Name=dyn_jacket_b_r_02, swing1LimitDegrees=15, swing2LimitDegrees=70, twistLimitDegrees=3}
 {bone1Name=dyn_jacket_b_r_02, bone2Name=dyn_jacket_b_r_01, swing1LimitDegrees=15, swing2LimitDegrees=60, twistLimitDegrees=3}
 {bone1Name=dyn_jacket_b_r_01, bone2Name=dyn_mstr_jacket_b_r, swing1LimitDegrees=10, swing2LimitDegrees=20, twistLimitDegrees=3}
 {bone1Name=dyn_jacket_f_l_04, bone2Name=dyn_jacket_f_l_03, swing1LimitDegrees=15, swing2LimitDegrees=35, twistLimitDegrees=3}
 {bone1Name=dyn_jacket_f_l_03, bone2Name=dyn_jacket_f_l_02, swing1LimitDegrees=20, swing2LimitDegrees=35, twistLimitDegrees=3}
 {bone1Name=dyn_jacket_f_l_02, bone2Name=dyn_jacket_f_l_01, swing1LimitDegrees=20, swing2LimitDegrees=35, twistLimitDegrees=3}
 {bone1Name=dyn_jacket_f_r_04, bone2Name=dyn_jacket_f_r_03, swing1LimitDegrees=20, swing2LimitDegrees=20, twistLimitDegrees=3}
 {bone1Name=dyn_jacket_f_r_03, bone2Name=dyn_jacket_f_r_02, swing1LimitDegrees=20, swing2LimitDegrees=20, twistLimitDegrees=3}
 {bone1Name=dyn_jacket_f_r_02, bone2Name=dyn_jacket_f_r_01, swing1LimitDegrees=20, swing2LimitDegrees=15, twistLimitDegrees=3}
 {bone1Name=dyn_jacket_f_r_01, bone2Name=dyn_mstr_jacket_f_r, swing1LimitDegrees=20, swing2LimitDegrees=15, twistLimitDegrees=3}
 {bone1Name=dyn_jacket_s_l_04, bone2Name=dyn_jacket_s_l_03, swing1LimitDegrees=15, swing2LimitDegrees=45, twistLimitDegrees=3}
 {bone1Name=dyn_jacket_s_l_03, bone2Name=dyn_jacket_s_l_02, swing1LimitDegrees=20, swing2LimitDegrees=30, twistLimitDegrees=3}
 {bone1Name=dyn_jacket_s_l_02, bone2Name=dyn_jacket_s_l_01, swing1LimitDegrees=20, swing2LimitDegrees=20, twistLimitDegrees=3}
 {bone1Name=dyn_jacket_s_l_01, bone2Name=dyn_mstr_jacket_s_l, swing1LimitDegrees=20, swing2LimitDegrees=20, twistLimitDegrees=3}
 {bone1Name=dyn_jacket_s_r_04, bone2Name=dyn_jacket_s_r_03, swing1LimitDegrees=25, swing2LimitDegrees=45, twistLimitDegrees=3}
 {bone1Name=dyn_jacket_s_r_03, bone2Name=dyn_jacket_s_r_02, swing1LimitDegrees=25, swing2LimitDegrees=35, twistLimitDegrees=3}
 {bone1Name=dyn_jacket_s_r_02, bone2Name=dyn_jacket_s_r_01, swing1LimitDegrees=20, swing2LimitDegrees=25, twistLimitDegrees=3}
 {bone1Name=dyn_jacket_s_r_01, bone2Name=dyn_mstr_jacket_s_r, swing1LimitDegrees=20, swing2LimitDegrees=20, twistLimitDegrees=3}
 {
  bone1Name=thigh_l
  bone2Name=pelvis
  swing1LimitDegrees=45
  swing1Motion=Free
  swing2LimitDegrees=45
  swing2Motion=Free
  twistLimitDegrees=45
  twistMotion=Free
 }
 {
  bone1Name=thigh_r
  bone2Name=pelvis
  swing1LimitDegrees=45
  swing1Motion=Free
  swing2LimitDegrees=45
  swing2Motion=Free
  twistLimitDegrees=45
  twistMotion=Free
 }
 {
  bone1Name=calf_l
  bone2Name=thigh_l
  swing1LimitDegrees=45
  swing1Motion=Free
  swing2LimitDegrees=45
  swing2Motion=Free
  twistLimitDegrees=45
  twistMotion=Free
 }
 {
  bone1Name=calf_r
  bone2Name=thigh_r
  swing1LimitDegrees=45
  swing1Motion=Free
  swing2LimitDegrees=45
  swing2Motion=Free
  twistLimitDegrees=45
  twistMotion=Free
 }
 {
  bone1Name=dyn_mstr_jacket_b_r
  bone2Name=pelvis
  swing1LimitDegrees=45
  swing1Motion=Free
  swing2LimitDegrees=45
  swing2Motion=Free
  twistLimitDegrees=45
  twistMotion=Free
 }
 {
  bone1Name=dyn_mstr_jacket_b_l
  bone2Name=pelvis
  swing1LimitDegrees=45
  swing1Motion=Free
  swing2LimitDegrees=45
  swing2Motion=Free
  twistLimitDegrees=45
  twistMotion=Free
 }
 {
  bone1Name=dyn_mstr_jacket_f_l
  bone2Name=pelvis
  swing1LimitDegrees=45
  swing1Motion=Free
  swing2LimitDegrees=45
  swing2Motion=Free
  twistLimitDegrees=45
  twistMotion=Free
 }
 {
  bone1Name=dyn_mstr_jacket_f_r
  bone2Name=pelvis
  swing1LimitDegrees=45
  swing1Motion=Free
  swing2LimitDegrees=45
  swing2Motion=Free
  twistLimitDegrees=45
  twistMotion=Free
 }
 {
  bone1Name=dyn_mstr_jacket_s_l
  bone2Name=pelvis
  swing1LimitDegrees=45
  swing1Motion=Free
  swing2LimitDegrees=45
  swing2Motion=Free
  twistLimitDegrees=45
  twistMotion=Free
 }
 {
  bone1Name=dyn_mstr_jacket_s_r
  bone2Name=pelvis
  swing1LimitDegrees=45
  swing1Motion=Free
  swing2LimitDegrees=45
  swing2Motion=Free
  twistLimitDegrees=45
  twistMotion=Free
 }
 {
  bone1Name=foot_r
  bone2Name=calf_r
  swing1LimitDegrees=45
  swing1Motion=Free
  swing2LimitDegrees=45
  swing2Motion=Free
  twistLimitDegrees=45
  twistMotion=Free
 }
 {
  bone1Name=foot_l
  bone2Name=calf_l
  swing1LimitDegrees=45
  swing1Motion=Free
  swing2LimitDegrees=45
  swing2Motion=Free
  twistLimitDegrees=45
  twistMotion=Free
 }
 {bone1Name=dyn_jacket_b_02, bone2Name=dyn_jacket_b_01, swing1LimitDegrees=20, swing2LimitDegrees=20, twistLimitDegrees=3}
 {bone1Name=dyn_jacket_b_01, bone2Name=dyn_mstr_jacket_b, swing1LimitDegrees=20, swing2LimitDegrees=20, twistLimitDegrees=3}
 {
  bone1Name=dyn_mstr_jacket_b
  bone2Name=pelvis
  swing1LimitDegrees=45
  swing1Motion=Free
  swing2LimitDegrees=45
  swing2Motion=Free
  twistLimitDegrees=45
  twistMotion=Free
 }
 {bone1Name=dyn_jacket_b_03, bone2Name=dyn_jacket_b_02, swing1LimitDegrees=30, swing2LimitDegrees=30, twistLimitDegrees=3}
 {bone1Name=dyn_jacket_b_04, bone2Name=dyn_jacket_b_03, swing1LimitDegrees=45, swing2LimitDegrees=45, twistLimitDegrees=3}
 {
  bone1Name=pelvis
  bone2Name=root
  swing1LimitDegrees=45
  swing1Motion=Free
  swing2LimitDegrees=45
  swing2Motion=Free
  twistLimitDegrees=45
  twistMotion=Free
 }
 {
  bone1Name=dyn_jacket_b_l_04
  bone2Name=dyn_jacket_b_04
  swing1LimitDegrees=45
  swing1Motion=Free
  swing2LimitDegrees=45
  swing2Motion=Free
  twistLimitDegrees=45
  twistMotion=Free
 }
 {
  bone1Name=dyn_jacket_s_l_04
  bone2Name=dyn_jacket_b_l_04
  swing1LimitDegrees=45
  swing1Motion=Free
  swing2LimitDegrees=45
  swing2Motion=Free
  twistLimitDegrees=45
  twistMotion=Free
 }
 {
  bone1Name=dyn_jacket_f_l_02
  bone2Name=dyn_jacket_s_l_02
  swing1LimitDegrees=45
  swing1Motion=Free
  swing2LimitDegrees=45
  swing2Motion=Free
  twistLimitDegrees=45
  twistMotion=Free
 }
 {
  bone1Name=dyn_jacket_b_03
  bone2Name=dyn_jacket_b_r_03
  swing1LimitDegrees=45
  swing1Motion=Free
  swing2LimitDegrees=45
  swing2Motion=Free
  twistLimitDegrees=45
  twistMotion=Free
 }
 {
  bone1Name=dyn_jacket_s_r_02
  bone2Name=dyn_jacket_b_r_02
  swing1LimitDegrees=45
  swing1Motion=Free
  swing2LimitDegrees=45
  swing2Motion=Free
  twistLimitDegrees=45
  twistMotion=Free
 }
 {
  bone1Name=dyn_jacket_f_r_02
  bone2Name=dyn_jacket_s_r_02
  swing1LimitDegrees=45
  swing1Motion=Free
  swing2LimitDegrees=45
  swing2Motion=Free
  twistLimitDegrees=45
  twistMotion=Free
 }
 {
  bone1Name=dyn_jacket_f_r_03
  bone2Name=dyn_jacket_s_r_03
  swing1LimitDegrees=45
  swing1Motion=Free
  swing2LimitDegrees=45
  swing2Motion=Free
  twistLimitDegrees=45
  twistMotion=Free
 }
 {
  bone1Name=dyn_jacket_f_l_03
  bone2Name=dyn_jacket_s_l_03
  swing1LimitDegrees=45
  swing1Motion=Free
  swing2LimitDegrees=45
  swing2Motion=Free
  twistLimitDegrees=45
  twistMotion=Free
 }
 {
  bone1Name=dyn_jacket_b_r_04
  bone2Name=dyn_jacket_s_r_04
  swing1LimitDegrees=45
  swing1Motion=Free
  swing2LimitDegrees=45
  swing2Motion=Free
  twistLimitDegrees=45
  twistMotion=Free
 }
 {bone1Name=dyn_jacket_f_l_01, bone2Name=dyn_mstr_jacket_f_l, swing1LimitDegrees=20, swing2LimitDegrees=25, twistLimitDegrees=3}
 {
  bone1Name=dyn_jacket_f_r_04
  bone2Name=dyn_jacket_s_r_04
  swing1LimitDegrees=45
  swing1Motion=Free
  swing2LimitDegrees=45
  swing2Motion=Free
  twistLimitDegrees=45
  twistMotion=Free
 }
]
