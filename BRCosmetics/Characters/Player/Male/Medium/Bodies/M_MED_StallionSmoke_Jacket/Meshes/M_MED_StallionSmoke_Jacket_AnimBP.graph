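# Post-process style cosmetic AnimBP for the StallionSmoke male jacket, copying the attached parent's pose and adding secondary motion.
# A Control Rig takes crouch, slide and front-end inputs, then weighted constraints place drawstring master bones for chained, plane-limited AnimDynamics.
# Three serial RigidBody nodes use normal, skydive and front-end physics assets; update logic zeroes the normal one's alpha in skydive dive or front end.
# Some constraint and RigidBody nodes are unconnected.
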AnimBlueprint /BRCosmetics/Characters/Player/Male/Medium/Bodies/M_MED_StallionSmoke_Jacket/Meshes/M_MED_StallionSmoke_Jacket_AnimBP
target_skeleton=&/M_MED_StallionSmoke_Jacket_Skeleton
graphs [
 AnimGraph {nodes [
   "Constraint - dyn_mstr_string_b (PPPPPP)"!(
    ComponentPose
    Alpha=1
    ConstraintWeights_0=0.313
    ConstraintWeights_1:float=0.163
    ConstraintWeights_2:float=0.043
    ConstraintWeights_3:float=0.295
    ConstraintWeights_4:float=0.143
    ConstraintWeights_5:float=0.043
   ) -> (Pose) {
    alphaBoolBlend {blendOption=Linear}
    boneToModify {boneName=dyn_mstr_string_b}
    constraintSetup [
     {targetBone {boneName=dyn_jacket_b_r_04}}
     {targetBone {boneName=dyn_jacket_b_r_03}}
     {targetBone {boneName=dyn_jacket_b_r_02}}
     {targetBone {boneName=dyn_jacket_b_l_04}}
     {targetBone {boneName=dyn_jacket_b_l_03}}
     {targetBone {boneName=dyn_jacket_b_l_02}}
    ]
    constraintWeights [1,1,1,1,1,1]
   }
   n1 = "Constraint - dyn_jacket_b_04 (PP)"!(ComponentPose, Alpha=1, ConstraintWeights_0=0.9, ConstraintWeights_1:float=0.1) -> (Pose) {
    alphaBoolBlend {blendOption=Linear}
    boneToModify {boneName=dyn_jacket_b_04}
    constraintSetup [{targetBone {boneName=dyn_jacket_b_l_04}}, {targetBone {boneName=dyn_jacket_b_r_04}}]
    constraintWeights [1,1]
   }
   "Constraint - dyn_jacket_b_03 (PP)"!(ComponentPose=@n1, Alpha=1, ConstraintWeights_0=0.9, ConstraintWeights_1:float=0.1) -> (Pose) {
    alphaBoolBlend {blendOption=Linear}
    boneToModify {boneName=dyn_jacket_b_03}
    constraintSetup [{targetBone {boneName=dyn_jacket_b_l_03}}, {targetBone {boneName=dyn_jacket_b_r_03}}]
    constraintWeights [1,1]
   }
   RigidBody() {
    bClampLinearTranslationLimitToRefPose=true
    bOverrideWorldGravity=true
    overridePhysicsAsset=&/M_MED_StallionSmoke_Jacket_Physics
    simulationSpace=BaseBoneSpace
    alphaBoolBlend {blendOption=Linear}
    baseBoneRef {boneName=pelvis}
    componentAppliedLinearAccClamp=<3000 3000 3000>
    componentLinearAccScale=<0.015 0.015 0.02>
    componentLinearVelScale=<0.01 0.01 0.02>
    overrideWorldGravity {z=-980}
    simSpaceSettings {velocityScaleZ=0.7, worldAlpha=0.3}
   }
   n2 = Control Rig!(
    Source=Copy Pose From Mesh() {bCopyCurves=true, bUseAttachedParent=true}
    Alpha
    CrouchCtrl:bool=Get bIsCrouching()
    CrouchAlphaCtrl:double=Get JacketCrouchAlpha() -> (JacketCrouchAlpha:double)
    SlideCtrl:double=Get JacketSlideAlpha() -> (JacketSlideAlpha:double)
    FrontEndLacesAlphaCtrl:float="To Float (Boolean)"(InBool=Get bIsInFrontEnd())
   ) -> (Pose) {
    alphaBoolBlend {blendOption=Linear}
    controlRigAssetReference {blueprintRigClass=&/M_MED_StallionSmoke_Jacket_CtrlRig.M_MED_StallionSmoke_Jacket_CtrlRig_C}
   }
   n3 = "Constraint - dyn_mstr_string_b (PP)"!(ComponentPose=Local To Component(LocalPose=@n2), Alpha=1, ConstraintWeights_0=0.89, ConstraintWeights_1:float=0.11) -> (Pose) {
    alphaBoolBlend {blendOption=Linear}
    boneToModify {boneName=dyn_mstr_string_b}
    constraintSetup [{targetBone {boneName=dyn_jacket_b_04}}, {targetBone {boneName=dyn_jacket_b_r_04}}]
    constraintWeights [1,1]
   }
   n4 = "Constraint - dyn_mstr_string_l (PP)"!(ComponentPose=@n3, Alpha=1, ConstraintWeights_0=0.49, ConstraintWeights_1:float=0.51) -> (Pose) {
    alphaBoolBlend {blendOption=Linear}
    boneToModify {boneName=dyn_mstr_string_l}
    constraintSetup [{targetBone {boneName=spine_05}}, {targetBone {boneName=dyn_collar_l}}]
    constraintWeights [1,1]
   }
   n5 = "Constraint - dyn_mstr_string_r (PPPPP)"!(
    ComponentPose=@n4
    Alpha=1
    ConstraintWeights_0=0.49
    ConstraintWeights_1:float=0.37
    ConstraintWeights_2:float=0.06
    ConstraintWeights_3:float=0.03
    ConstraintWeights_4:float=0.05
   ) -> (Pose) {
    alphaBoolBlend {blendOption=Linear}
    boneToModify {boneName=dyn_mstr_string_r}
    constraintSetup [
     {targetBone {boneName=spine_05}}
     {targetBone {boneName=dyn_collar_r}}
     {targetBone {boneName=spine_04}}
     {targetBone {boneName=spine_03}}
     {targetBone {boneName=clavicle_r}}
    ]
    constraintWeights [1,1,1,1,1]
   }
   n6 = Anim Dynamics - Bone: dyn_string_l_01(ComponentPose=@n5, Alpha=1, ExternalForce="0,0,0") {
    angularDampingOverride=0.9999
    bOverrideAngularDamping=true
    simulationSpace=RootRelative
    alphaBoolBlend {blendOption=Linear}
    boundBone {boneName=dyn_string_l_01}
    physicsBodyDefinitions [{
      boundBone {boneName=dyn_string_l_01}
      boxExtents=<40 40 40>
      constraintSetup {angularLimitsMax=<0 5 10>, angularLimitsMin=<0 -50 -30>, angularTarget=<0 0 0>}
      localJointOffset=<1 0 0>
     }]
    relativeSpaceBone {boneName=spine_05}
   }
   n7 = Anim Dynamics - Bone: dyn_string_l_02(ComponentPose=@n6, Alpha=1, ExternalForce="0,0,0") {
    angularDampingOverride=0.9999
    bOverrideAngularDamping=true
    simulationSpace=RootRelative
    alphaBoolBlend {blendOption=Linear}
    boundBone {boneName=dyn_string_l_02}
    physicsBodyDefinitions [{
      boundBone {boneName=dyn_string_l_02}
      boxExtents=<15 15 15>
      constraintSetup {angularLimitsMax=<0 15 10>, angularLimitsMin=<0 -50 -30>, angularTarget=<0 0 0>}
      localJointOffset=<0.3 0 0>
     }]
    planarLimits [
     {drivingBone {boneName=spine_05}, planeTransform {location=<15 0 0>, rotation=<rot 0 -30 90>, scale=<0.2 0.2 0.2>}}
     {drivingBone {boneName=spine_05}, planeTransform {location=<0 0 -10>, rotation=<rot 5 15 170>}}
    ]
    relativeSpaceBone {boneName=spine_05}
   }
   n8 = Anim Dynamics - Bone: dyn_string_l_03(ComponentPose=@n7, Alpha=1, ExternalForce="0,0,0") {
    angularDampingOverride=0.9999
    bOverrideAngularDamping=true
    simulationSpace=RootRelative
    alphaBoolBlend {blendOption=Linear}
    boundBone {boneName=dyn_string_l_03}
    physicsBodyDefinitions [{
      boundBone {boneName=dyn_string_l_03}
      boxExtents=<5 5 5>
      constraintSetup {angularLimitsMax=<0 30 20>, angularLimitsMin=<0 -50 -30>, angularTarget=<0 0 0>}
      localJointOffset=<0.3 0 0>
     }]
    planarLimits [
     {drivingBone {boneName=spine_05}, planeTransform {location=<0 15 -10>, rotation=<rot 0 -45 90>}}
     {drivingBone {boneName=spine_05}, planeTransform {location=<0 0 -11.5>, rotation=<rot 5 15 170>}}
    ]
    relativeSpaceBone {boneName=spine_05}
   }
   n9 = Anim Dynamics - Bone: dyn_string_r_01(ComponentPose=@n8, Alpha=1, ExternalForce="0,0,0") {
    angularDampingOverride=0.9999
    bOverrideAngularDamping=true
    simulationSpace=RootRelative
    alphaBoolBlend {blendOption=Linear}
    boundBone {boneName=dyn_string_r_01}
    physicsBodyDefinitions [{
      boundBone {boneName=dyn_string_r_01}
      boxExtents=<40 40 40>
      constraintSetup {angularLimitsMax=<0 5 30>, angularLimitsMin=<0 -50 -10>, angularTarget=<0 0 0>}
      localJointOffset=<1 0 0>
     }]
    relativeSpaceBone {boneName=spine_05}
   }
   n10 = Anim Dynamics - Bone: dyn_string_r_02(ComponentPose=@n9, Alpha=1, ExternalForce="0,0,0") {
    angularDampingOverride=0.9999
    bOverrideAngularDamping=true
    simulationSpace=RootRelative
    alphaBoolBlend {blendOption=Linear}
    boundBone {boneName=dyn_string_r_02}
    physicsBodyDefinitions [{
      boundBone {boneName=dyn_string_r_02}
      boxExtents=<15 15 15>
      constraintSetup {angularLimitsMax=<0 15 30>, angularLimitsMin=<0 -50 -10>, angularTarget=<0 0 0>}
      localJointOffset=<0.3 0 0>
     }]
    planarLimits [{drivingBone {boneName=spine_05}, planeTransform {location=<23 0 12>, rotation=<rot 0 -30 80>}}]
    relativeSpaceBone {boneName=spine_05}
   }
   n11 = Anim Dynamics - Bone: dyn_string_r_03(ComponentPose=@n10, Alpha=1, ExternalForce="0,0,0") {
    angularDampingOverride=0.9999
    bOverrideAngularDamping=true
    simulationSpace=RootRelative
    alphaBoolBlend {blendOption=Linear}
    boundBone {boneName=dyn_string_r_03}
    physicsBodyDefinitions [{
      boundBone {boneName=dyn_string_r_03}
      boxExtents=<5 5 5>
      constraintSetup {angularLimitsMax=<0 30 30>, angularLimitsMin=<0 -50 -15>, angularTarget=<0 0 0>}
      localJointOffset=<0.3 0 0>
     }]
    planarLimits [{drivingBone {boneName=spine_05}, planeTransform {location=<21 0 12>, rotation=<rot 0 -25 80>}}]
    relativeSpaceBone {boneName=spine_05}
   }
   n12 = Anim Dynamics - Bone: dyn_string_b_01(ComponentPose=@n11, Alpha=1, ExternalForce="0,0,0") {
    angularDampingOverride=0.9999
    bOverrideAngularDamping=true
    simulationSpace=RootRelative
    alphaBoolBlend {blendOption=Linear}
    boundBone {boneName=dyn_string_b_01}
    physicsBodyDefinitions [{
      boundBone {boneName=dyn_string_b_01}
      boxExtents=<5 5 5>
      constraintSetup {angularLimitsMax=<0 50 50>, angularLimitsMin=<0 -50 -50>, angularTarget=<0 0 0>}
      localJointOffset=<0.2 0 0>
     }]
    relativeSpaceBone {boneName=dyn_jacket_b_04}
   }
   n13 = RigidBody(ComponentPose=@n12, Alpha=Get NormalRbanAlpha() -> (NormalRbanAlpha:double)) {
    bClampLinearTranslationLimitToRefPose=true
    bOverrideWorldGravity=true
    overridePhysicsAsset=&M_MED_StallionSmoke_Jacket_Physics
    simulationSpace=BaseBoneSpace
    alphaBoolBlend {blendOption=Linear}
    baseBoneRef {boneName=pelvis}
    componentAppliedLinearAccClamp=<3000 3000 3000>
    componentLinearAccScale=<1 1 0.3>
    componentLinearVelScale=<1 1 0.3>
    overrideWorldGravity {z=-980}
    simSpaceSettings {maxAngularVelocity=0, maxLinearAcceleration=700, maxLinearVelocity=500, velocityScaleZ=0.1, worldAlpha=0.7, externalLinearDragV=<0.1 0.1>}
   }
   n14 = RigidBody(ComponentPose=@n13) {
    bClampLinearTranslationLimitToRefPose=true
    bOverrideWorldGravity=true
    overridePhysicsAsset=&/M_MED_StallionSmoke_Jacket_SkyDive_Physics
    simulationSpace=BaseBoneSpace
    alphaBoolBlend {blendOption=Linear}
    baseBoneRef {boneName=pelvis}
    componentAppliedLinearAccClamp=<3000 3000 3000>
    componentLinearAccScale=<1 1 0.3>
    componentLinearVelScale=<1 1 0.3>
    overrideWorldGravity {z=-980}
    simSpaceSettings {maxAngularVelocity=0, maxLinearAcceleration=700, maxLinearVelocity=500, velocityScaleZ=0.1, worldAlpha=0.7, externalLinearDragV=<0.1 0.1>}
   }
   n15 = RigidBody(ComponentPose=@n14) {
    bClampLinearTranslationLimitToRefPose=true
    bOverrideWorldGravity=true
    overridePhysicsAsset=&/M_MED_StallionSmoke_Jacket_FE_Physics
    simulationSpace=BaseBoneSpace
    alphaBoolBlend {blendOption=Linear}
    baseBoneRef {boneName=pelvis}
    componentAppliedLinearAccClamp=<3000 3000 3000>
    componentLinearAccScale=<1 1 0.3>
    componentLinearVelScale=<1 1 0.3>
    overrideWorldGravity {z=-980}
    simSpaceSettings {maxAngularVelocity=0, maxLinearAcceleration=700, maxLinearVelocity=500, velocityScaleZ=0.1, worldAlpha=0.7, externalLinearDragV=<0.1 0.1>}
   }
   Output Pose(Result=Component To Local(ComponentPose=@n15)) {layerGroup=None}
   Get bIsSliding()
  ]}
 EventGraph {nodes [
   Event BlueprintUpdateAnimation()
   > Set JacketCrouchAlpha(JacketCrouchAlpha:double=SelectFloat(A=1, B=0, bPickA=Get bIsCrouching())) -> (Output_Get:double)
   n1 = Get bIsSkydiveDiveMode()
   n2 = Get bIsInFrontEnd()
   > Set NormalRbanAlpha(NormalRbanAlpha:double=SelectFloat(A=0, B=1, bPickA=OR Boolean(A=@n1, B=@n2))) -> (Output_Get:double)
   > Set JacketSlideAlpha(JacketSlideAlpha:double=SelectFloat(A=1, B=0, bPickA=Get bIsSliding())) -> (Output_Get:double)
   TryGetPawnOwner()
  ]}
]
variables [
 JacketCrouchAlpha {type=number}
 JacketSlideAlpha {type=number}
 NormalRbanAlpha {type=number}
]
